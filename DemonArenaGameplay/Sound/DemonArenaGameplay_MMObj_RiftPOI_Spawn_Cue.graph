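# Sound cue for the spawn of a Rift POI in Demon Arena gameplay, on the Fort_Gadgets sound class.
# A mixer layers a close-attenuation branch with a louder distant-attenuation branch delayed by a fixed one second.
# Both attenuation nodes have empty inputs, so the graph references no sound waves.
# The cue overrides concurrency to cap simultaneous instances and carries a very large max distance.
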SoundCue /DemonArenaGameplay/Sound/DemonArenaGameplay_MMObj_RiftPOI_Spawn_Cue
root=Mixer_2
nodes [
 {id=Attenuation_0, children [null], properties {attenuationSettings=&/DemonArenaGameplay_MMObj_POI_Distant_Att}}
 {id=Attenuation_1, children [null], properties {attenuationSettings=&/DemonArenaGameplay_MMObj_POI_Close_Att}}
 {id=Delay_0, children [Attenuation_0], properties {delayMax=1, delayMin=1}}
 {id=Mixer_2, children [Attenuation_1, Delay_0], properties {inputVolume [1,1.25]}}
]
properties {
 bOverrideConcurrency=true
 maxDistance=350000
 soundClassObject=&Fort_Gadgets
 volumeMultiplier=0.7
 concurrencyOverrides {maxCount=6}
}
